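# Physics asset for the Mad Commander backpack cosmetic, setting up collision bodies and joints for its dangling charm, hanging rabbit and top flap.
# A root box anchors three branches: a two-link charm, the flap, and a two-link chain ending in the rabbit head.
# The rabbit's ears and body hang off the head.
# Most joints swing widely with little twist, while the flap is held to very tight limits.
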PhysicsAsset /BRCosmetics/Accessories/FORT_Backpacks/F_MED_Mad_Commander_Backpack/Meshes/F_MED_Mad_Commander_Backpack_Dyn
bodies {
 dyn_charm_01 [{shapeType=Capsule, length=1.25, radius=0.75, center=<0.8776 0 0>, rotation=<rot 90 0 0>}]
 dyn_charm_02 [{shapeType=Box, extentX=4.5, extentY=4.5, extentZ=2.5, center=<2.0385 0 -0.0001>}]
 dyn_rabbit_body [{shapeType=Capsule, length=0.5312, radius=1.8438, center=<1.6706 0 0.4867>, rotation=<rot 90 0 0>}]
 dyn_rabbit_chain_01 [{shapeType=Capsule, length=1, radius=1, center=<1 0 0>, rotation=<rot 90 0 0>}]
 dyn_rabbit_chain_02 [{shapeType=Capsule, length=1, radius=1, center=<1 0 0>, rotation=<rot 90 0 0>}]
 dyn_rabbit_ear_l [{shapeType=Capsule, length=3.4062, radius=1, center=<-0.2567 0 0>, rotation=<rot 90 0 0>}]
 dyn_rabbit_ear_r [{shapeType=Capsule, length=3.4062, radius=1, center=<-0.2567 0 0>, rotation=<rot 90 0 0>}]
 dyn_rabbit_head [{shapeType=Capsule, length=0, radius=2.5, center=<1.657 0.1541 1.144>, rotation=<rot 90 0 0>}]
 dyn_top_flap [{shapeType=Capsule, length=2, radius=2, center=<1.6239 0 0>, rotation=<rot 90 0 0>}]
 root [{shapeType=Box, extentX=7.5, extentY=13.75, extentZ=7.5, center=<-8.6801 -5.2486 -16.028>, rotation=<rot 0 10.0018 0>}]
}
constraints [
 {bone1Name=dyn_charm_01, bone2Name=root, swing1LimitDegrees=45, swing2LimitDegrees=45, twistLimitDegrees=5}
 {bone1Name=dyn_charm_02, bone2Name=dyn_charm_01, swing1LimitDegrees=45, swing2LimitDegrees=45, twistLimitDegrees=5}
 {bone1Name=dyn_top_flap, bone2Name=root, swing1LimitDegrees=2, swing2LimitDegrees=7, twistLimitDegrees=2}
 {bone1Name=dyn_rabbit_chain_01, bone2Name=root, swing1LimitDegrees=45, swing2LimitDegrees=45, twistLimitDegrees=5}
 {bone1Name=dyn_rabbit_chain_02, bone2Name=dyn_rabbit_chain_01, swing1LimitDegrees=45, swing2LimitDegrees=45, twistLimitDegrees=5}
 {bone1Name=dyn_rabbit_head, bone2Name=dyn_rabbit_chain_02, swing1LimitDegrees=45, swing2LimitDegrees=45, twistLimitDegrees=5}
 {bone1Name=dyn_rabbit_ear_r, bone2Name=dyn_rabbit_head, swing1LimitDegrees=45, swing2LimitDegrees=45, twistLimitDegrees=5}
 {bone1Name=dyn_rabbit_ear_l, bone2Name=dyn_rabbit_head, swing1LimitDegrees=45, swing2LimitDegrees=45, twistLimitDegrees=5}
 {bone1Name=dyn_rabbit_body, bone2Name=dyn_rabbit_head, swing1LimitDegrees=45, swing2LimitDegrees=45, twistLimitDegrees=5}
]
